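# Conversation database for Season 30 Story P03 Quest 06, where the player gives Hope the Pandora's Gem and hears her gem vision of Megalo Don and a cloaked Wanderer.
# One entry tag leads into a linear, branchless chain of speech nodes.
# The opening choice auto-selects if valid, gated on the NPC's VoidRedemption tag, the active quest and its incomplete obj0 objective.
# The gem hand-over choice carries the quest's gameplay tag.
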ConversationDatabase /BattlePassS30_Quests/Conversations/Story/P03/Quest_06/Conversation_S30_Story_VR_P03_Q06
entry_points [{tag=Conversation.Entry.Season30.Story.P03.Q06.VoidRedemption, destinations [#1]}]
nodes [
 ConversationEntryPointNode_0 {
  id=#1
  class=ConversationEntryPointNode
  outputs [#2]
  properties {entryTag {tagName=Conversation.Entry.Season30.Story.P03.Q06.VoidRedemption}}
 }
 Speech_C_1 {
  id=#2
  class=Speech_C
  outputs [#3]
  properties {
   softTaskIcon=&T_UI_NPC_MainQuest
   generalConfig {defaultMessage="That's two gems down, and according to my intel, two to go. Nice work."}
  }
  sub_nodes [
   ConversationChoiceNode_2 {
    class=ConversationChoiceNode
    properties {
     defaultChoiceDisplayText=Talk to Hope
     choiceTags {gameplayTags [{tagName=Conversation.Choice.AutoSelectIfValid}, {tagName=Conversation.Choice.Category.MainQuestline}]}
    }
   }
   FortConversationRequirement_ControllerRequirement_1 {
    class=FortConversationRequirement_ControllerRequirement
    properties {
     nodeName=Require: NPC Has Conversation Tag
     participantId {tagName=Conversation.Participant.NPC}
     requirement {class=FortControllerRequirement_NPCHasConversationEntryTag, properties {queryTag {tagName=Conversation.Entry.VoidRedemption}}}
    }
   }
   FortConversationRequirement_ControllerRequirement_3 {
    class=FortConversationRequirement_ControllerRequirement
    properties {
     nodeName=Require: Is Quest Active
     requirement {
      class=FortControllerRequirement_IsQuestActive
      properties {questDefinition=/BattlePassS30_Quests/Items/QuestItems/Story/P03/Quest_S30_Story_P03_Q06}
     }
    }
   }
   FortConversationRequirement_ControllerRequirement_5 {
    class=FortConversationRequirement_ControllerRequirement
    properties {
     nodeName=Require: Has Completed Quest Objective
     requirement {
      class=FortControllerRequirement_HasCompletedQuestObjective
      properties {
       bInverseCondition=true
       objectiveBackendName=Quest_S30_Story_P03_Q06_obj0
       questDefinition=&Quest_S30_Story_P03_Q06
      }
     }
    }
   }
  ]
 }
 Speech_C_28 {
  id=#3
  class=Speech_C
  outputs [#4]
  properties {
   softTaskIcon=&T_UI_NPC_Conversation_Give
   generalConfig {defaultMessage="While I have you... It's probably nothing... But, um, just between us..."}
  }
  sub_nodes [ConversationChoiceNode_1 {
    class=ConversationChoiceNode
    properties {
     defaultChoiceDisplayText=Give Hope the Pandora's Gem
     choiceTags {gameplayTags [{tagName=Athena.Quests.S30.Conversation.Story.P03.06.VoidRedemption}]}
    }
   }]
 }
 Speech_C_2 {
  id=#4
  class=Speech_C
  outputs [#5]
  properties {generalConfig {defaultMessage="I've always had \"artistic vision\" but the gems... show me things."}}
  sub_nodes [ConversationChoiceNode_3 {class=ConversationChoiceNode, properties {defaultChoiceDisplayText=Continue}}]
 }
 Speech_C_0 {
  id=#5
  class=Speech_C
  outputs [#6]
  properties {generalConfig {defaultMessage="They showed me Megalo Don working with a cloaked Wanderer. "}}
  sub_nodes [ConversationChoiceNode_0 {class=ConversationChoiceNode, properties {defaultChoiceDisplayText=Continue}}]
 }
 Speech_C_3 {
  id=#6
  class=Speech_C
  outputs [#7]
  properties {generalConfig {defaultMessage="I haven't seen the guy, so maybe it's a metaphor or something."}}
  sub_nodes [ConversationChoiceNode_5 {class=ConversationChoiceNode, properties {defaultChoiceDisplayText=Continue}}]
 }
 Speech_C_4 {
  id=#7
  class=Speech_C
  outputs []
  properties {generalConfig {defaultMessage="Watch your back out there, okay? And if you see a big guy in weather-inappropriate robes, let me know."}}
  sub_nodes [ConversationChoiceNode_4 {class=ConversationChoiceNode, properties {defaultChoiceDisplayText=Continue}}]
 }
]
